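; Quest gameplay ability for vehicle-distance quests (Octopus, by its name) that flags when the player's vehicle exceeds a speed limit.
; Event activation starts a looping timer with a randomized interval that reads the avatar's vehicle forward speed in km/h.
; At or above MaxSpeedLimit it applies a gameplay effect to the owner, otherwise it removes effects granting the SpeedLimit quest tag.
; Ending the ability clears the timer.
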(event Ability|EventActivateAbilityFromEvent
  (params (EventData :type Gameplay%20Event%20Data%20Structure%20%28by%20ref%29))
  (|CallParentFunction EventData)
  (bind _returnvalue (Utilities|Time|SetTimerbyEvent (delegate Custom|Check%20Speed) (Math|Random|RandomFloatinRange 1.7 2.5) true))
  (Variables|Default|SetTimer%20Handle _returnvalue))

(event Custom|Check%20Speed
  ;[1 1 1] Check if player is over the speed
  (bind _asfort_athena_vehicle (Utilities|Casting|CastToFortAthenaVehicle (Vehicle|GetVehicleInterface (CallFunction|GetAvatarasFortPlayerPawn))))
  (if (>= (Vehicle|GetForwardSpeedKmh _asfort_athena_vehicle) (Variables|Default|GetMaxSpeedLimit))
    ;[1 1 1] Apply or Remove Tag
    (Ability|ApplyGameplayEffectToOwner (Variables|Default|GetGameplay%20Effect))
    (else
      (Ability|RemoveGameplayEffectFromOwnerWithGrantedTags "(GameplayTags=((TagName=\"Granted.Athena.Quest.SpeedLimit\")))"))))

(event Ability|EventOnEndAbility
  (params (bWasCancelled :type Boolean))
  ;[1 1 1] Event End
  (|CallParentFunction bWasCancelled)
  (if (Utilities|Time|IsValidTimerHandle (Variables|Default|GetTimer%20Handle))
    (Utilities|Time|ClearandInvalidateTimerbyHandle (Variables|Default|GetTimer%20Handle))))
